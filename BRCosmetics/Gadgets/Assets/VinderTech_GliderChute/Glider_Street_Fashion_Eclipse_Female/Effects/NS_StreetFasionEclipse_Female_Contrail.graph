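# Niagara contrail system for the female Street Fashion Eclipse glider, drawing two ribbon anim trails between skeletal mesh bone or socket pairs.
# Trail spawning is gated on the glider being fully deployed.
# Alpha rises then fades over a curve, scaled in gameplay by a clamped system integrator value, while ribbons narrow and use camera fade.
# Front-end modifiers swap in different velocity, camera-fade offset and alpha for menu display.
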NiagaraSystem /BRCosmetics/Gadgets/Assets/VinderTech_GliderChute/Glider_Street_Fashion_Eclipse_Female/Effects/NS_StreetFasionEclipse_Female_Contrail
dependencies {
 usedDataInterfaces [&NiagaraDataInterfaceCamera, &NiagaraDataInterfaceCurve]
 usedDynamicInputs [&Power, &Divide_Float, &OneMinusFloat]
 usedModules [&FN_Integrator, &SystemState]
}
emitters {FN_AnimTrail_V4001 {
  emitterUpdateScript [
   &EmitterState()
   &FN_SpawnAnimTrail_v2(
    Number of Anim Trails=2
    Spawn Interval:NiagaraInt32=&MakeCustomIntFromBool(Bool=$User.bIsFullyDeployed, True INT=1, False INT=0)
   )
  ]
  particleSpawnScript [
   /Game/Effects/Niagara/Modules/AnimTrail/FN_InitializeAnimTrail/FN_InitializeAnimTrailBoneIndecies(Second Bone or Socket Index=2)
   &FN_InitializeAnimTrailBoneIndecies(AnimTrail Selection=1, First Bone or Socket Index=3, Second Bone or Socket Index=1) {module=FN_InitializeAnimTrailBoneIndecies001}
   /Game/Effects/Niagara/Modules/AnimTrail/FN_InitializeAnimTrail/FN_InitializeAnimTrail(AnimTrailWidth=3.5)
   &AddVelocity(Velocity=&FN_FrontEndVector3Modifier(Gameplay=<0 0 0>, Front End=<-500 -125 0>))
  ]
  particleUpdateScript [
   &self:CameraFade(
    FadeDistance=150
    Offset:NiagaraFloat=&FN_FrontEndFloatModifier(Gameplay=250, Front End=150)
    Particle Position:NiagaraPosition=$Particles.Position
   ) {module=CameraFade}
   &SolveForcesAndVelocity(Clamp Velocity=false, Limit Acceleration=false)
   /Game/Effects/Niagara/Modules/AnimTrail/FN_UpdateAnimTrail/FN_UpdateAnimTrail()
   &ScaleColor(
    Scale Alpha=&FloatFromCurve(
     FloatCurve=DataInterface() {propertyValues {Curve {keys [<at 0 0>,<at 0.25 1>,<at 1 0>]}}}
     Scale Curve=&FN_FrontEndFloatModifier(
      Gameplay=&ClampFloat(Float=&Multiply_Float(A=0.8, B=$System.FN_Integrator.Value), Max=0.35)
      Front End=0.15
     )
    )
    ScaleA=true
    ScaleRGB=false
   )
   &ScaleRibbonWidth(Ribbon Width Scale=&FloatFromCurve(FloatCurve=DataInterface() {propertyValues {Curve {keys [<at 0 1>,<at 0.9847 0.159>]}}}))
  ]
  rendererClasses [&NiagaraRibbonRendererProperties]
  renderers [{rendererClass=&NiagaraRibbonRendererProperties}]
 }}
summary {
 systemName=NS_StreetFasionEclipse_Female_Contrail
 emitters [{emitterName=FN_AnimTrail_V4001, rendererClasses [&NiagaraRibbonRendererProperties]}]
 userVariables [
  User.bisFrontEnd {type=NiagaraBool}
  User.bIsFrontEndPreview {type=NiagaraBool}
  User.bIsFullyDeployed {type=NiagaraBool}
  User.ForwardDot {type=NiagaraFloat, defaultValue=0, description=""}
  User.RibbonIDOffset {type=NiagaraInt32}
  User.SkeletalMesh {
   type=NiagaraDataInterfaceSkeletalMesh
   defaultValue {
    struct=NiagaraExt_VariableValue_DataInterface
    value {dataInterface=&self:NiagaraDataInterfaceSkeletalMesh_0}
   }
  }
 ]
}
